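# Material function for the TidalNinja character that blends a glowing, rough orb material over incoming attributes, masked by a TechArtMask channel.
# Inner orb detail comes from a refracted view ray through a normal-mapped surface, sampling panning noise in long-lat UVs.
# Triplanar-style smoke projects distorted, panning noise in pre-skinned local space, weighted by normals and colored through a curve atlas.
# Orb colors, a Fresnel term and a low-to-high emissive intensity range are also defined.
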MaterialFunction /BRCosmetics/Characters/Player/Female/Medium/Bodies/F_MED_TidalNinja/Materials/MF_TidalNinja
graph {nodes [
  n1 = SetMaterialAttributes!(
   MaterialAttributes
   EmissiveColor=NamedRerouteUsage()
   BaseColor
   Metallic=ScalarParameter() {parameterName=Metallic}
   Specular=ScalarParameter() {defaultValue=0.1, parameterName=Specular}
   Roughness=ScalarParameter() {defaultValue=0.9, parameterName=Roughness}
  ) {attributeSetTypes [
    #1
    #2
    #3
    #4
    #5
   ]}
  FunctionOutput(None=BlendMaterialAttributes(A=NamedRerouteUsage(), B=@n1, Alpha=NamedRerouteUsage()))
  LinearInterpolate(
   A=VectorParameter() {parameterName=OrbColor1, defaultValue=<rgba 0.0165 0.0028 0.0365 1>}.RGB
   B=VectorParameter() {parameterName=OrbColor2, defaultValue=<rgba 0.1855 0.0063 0.9167 1>}.RGB
   Alpha=NamedRerouteUsage()
  )
  n2 = CameraVectorWS()
  n3 = ScalarParameter() {defaultValue=2, group=TidalNinja, parameterName=CircumferenceTiling}
  MaterialExpressionVertexInterpolator_1 = VertexInterpolator(VS=LocalPosition() {localOrigin=InstancePreSkinning}.XYZ)
  n4 = Convert(None=@MaterialExpressionVertexInterpolator_1.PS) {
   nodeName=Break Float3
   convertInputs [{
     type=Vector3
     expressionInput {
      expression=&self:MaterialExpressionVertexInterpolator_1
      inputName=None
      mask=0
      maskA=0
      maskB=0
      maskG=0
      maskR=0
      outputIndex=0
     }
    }]
   convertOutputs [{type=Vector3}]
  }
  n5 = Multiply(
   A=ComponentMask(None=PreSkinnedLocalBounds().Extents) {b=true}
   B=ScalarParameter() {defaultValue=1, group=TidalNinja, parameterName=CircumferenceZHeight}
  )
  n6 = TextureSampleParameter2D(UVs=NamedRerouteDeclaration(None=MaterialFunctionCall!(Circumference Tiling=@n3, Position=@n4, Z Height=@n5, HQ Angle Calculation?) {materialFunction=/Game/Characters/Player/Common/MaterialFunctions/MF_CylindricalUVs_New}.UVs) {name=CylindtalUV, nodeColor=<rgba 1 0 0.1641 1>}) {
   desc=Rock/Gem normal map
   group=00_SuperLevels_T1/2
   parameterName=OrbNormal
   samplerSource=SSM_Wrap_WorldGroupSettings
   samplerType=SAMPLERTYPE_Normal
   texture=/BRCosmetics/Characters/Player/Female/Medium/Bodies/F_MED_RoyalAngst_Tote/Textures/T_RoughGemSurface_N
  }
  n7 = ScalarParameter() {defaultValue=0.7, group=TidalNinja, parameterName=OrbSurfaceNormalFlatness}
  n8 = MaterialFunctionCall!(Vector=MaterialFunctionCall!(
    Ray Direction=Multiply(A=@n2) {constB=-1}
    Surface Normal=MaterialFunctionCall(Normal=@n6.RGB, Flatness=@n7) {materialFunction=&FlattenNormal}.Result
    Refractive Index Origin
    Refractive Index Target=ScalarParameter() {defaultValue=1.05, group=TidalNinja, parameterName=OrbRefraction IOR}
   ) {materialFunction=/Engine/Functions/Engine_MaterialFunctions01/Vectors/Refract}.Vector) {materialFunction=/Engine/Functions/Engine_MaterialFunctions01/Coordinates/LongLatToUV}
  n9 = ScalarParameter() {defaultValue=4, group=TidalNinja, parameterName=OrbInnerNoise_Scale}
  n10 = Constant2Vector() {g=0.3, r=0.08}
  n11 = Multiply(A=Time(), B=ScalarParameter() {defaultValue=1, group=TidalNinja, parameterName=OrbInnerNoise_PanSpeed})
  n12 = TextureSample(UVs=Add(A=Multiply(A=@n8.Result, B=@n9), B=Multiply(A=@n10, B=@n11))) {
   desc=LowResBlurredNoiseNormalized
   samplerSource=SSM_Wrap_WorldGroupSettings
   samplerType=SAMPLERTYPE_Grayscale
   texture=/Game/Global/Textures/Default/Noise/LowResBlurredNoiseNormalized
  }
  NamedRerouteDeclaration(None=Power(Base=@n12.R, Exp=ScalarParameter() {defaultValue=1.5, group=TidalNinja, parameterName=OrbDetailPower})) {name=OrbDetail, nodeColor=<rgba 0 0.4844 1 1>}
  NamedRerouteDeclaration(None=ChannelMaskParameter(None=FunctionInput() {inputName=TechArtMask}) {parameterName=OrbChannel}.R) {name=OrbMask, nodeColor=<rgba 0.8047 1 0 1>}
  NamedRerouteDeclaration(None=FunctionInput() {inputName=MatIn, inputType=FunctionInput_MaterialAttributes}) {name=MatIn, nodeColor=<rgba 0.875 0 1 1>}
  n13 = LocalPosition() {includedOffsets=ExcludeOffsets, localOrigin=InstancePreSkinning}
  n14 = Multiply(
   A=ShaderStageSwitch(PixelShader=VertexInterpolator(VS=@n13.XYZ).PS, VertexShader=@n13.XYZ)
   B=ScalarParameter() {defaultValue=0.015, parameterName=TP_Tiling}
  )
  n15 = Multiply(A=Time(), B=VectorParameter() {parameterName=PanningEmissiveSpeed, defaultValue=<rgba 0.02 -0.03 -0.08 1>}.RGB)
  n16 = ScalarParameter() {defaultValue=0.3, parameterName=TP_DistortionAmount}
  n17 = Time()
  n18 = TextureSample(UVs=Add(A=NamedRerouteUsage(), B=Multiply(A=@n17) {constB=-0.005})) {
   samplerSource=SSM_Wrap_WorldGroupSettings
   texture=/BRCosmetics/Characters/Player/Female/Medium/Bodies/F_MED_TidalNinja/Textures/T_SWave_PerlinNoise_01_D
  }
  n19 = ConstantBiasScale(None=@n18.R) {bias=-0.5, scale=2}
  n20 = MaterialFunctionCall(Float3=Add(A=Add(A=@n14, B=@n15), B=Multiply(A=@n16, B=@n19))) {materialFunction=&BreakOutFloat3Components}
  n21 = TextureSample(UVs=MaterialFunctionCall(X=@n20.R, Y=@n20.R) {materialFunction=&MakeFloat2}.Result) {samplerSource=SSM_Wrap_WorldGroupSettings, texture=/Game/Effects/FogVolume/noise_flamy_Tile}
  n22 = PreSkinnedNormal()
  n23 = ShaderStageSwitch(PixelShader=VertexInterpolator(VS=@n22).PS, VertexShader=@n22)
  n24 = MaterialFunctionCall(Float3=Power(Base=Abs(None=@n23), Exp=ScalarParameter() {defaultValue=3, parameterName=TP_Power}) {constExponent=10}) {materialFunction=&BreakOutFloat3Components}
  n25 = TextureSample(UVs=MaterialFunctionCall(X=@n20.R, Y=@n20.R) {materialFunction=&MakeFloat2}.Result) {samplerSource=SSM_Wrap_WorldGroupSettings, texture=&noise_flamy_Tile}
  n26 = MaterialFunctionCall(X=@n20.G, Y=@n20.G) {materialFunction=&MakeFloat2}
  n27 = Add(
   A=Add(A=Multiply(A=@n21.B, B=@n24.B), B=Multiply(A=@n25.B, B=@n24.G))
   B=Multiply(A=TextureSample(UVs=@n26.Result) {samplerSource=SSM_Wrap_WorldGroupSettings, texture=&noise_flamy_Tile}.B, B=@n24.R)
  )
  n28 = CurveAtlasRowParameter(CurveTime=Power(Base=Saturate(None=@n27), Exp=ScalarParameter() {defaultValue=0.5, parameterName=TP_FinalPower}) {constExponent=1}) {
   atlas=&/CA_TidalNinja
   curve=&/CL_TidalNinja
   parameterName=SmokeCurve
  }
  NamedRerouteDeclaration(None=@n28) {name=TPSmoke, nodeColor=<rgba 0 0.5078 1 1>}
  n29 = LinearInterpolate(
   A=ScalarParameter() {defaultValue=1, parameterName=OrbEmissive_Low}
   B=ScalarParameter() {defaultValue=3, parameterName=OrbEmissive_High}
   Alpha=NamedRerouteUsage()
  ) {constA=0.6, constB=5}
  n30 = NamedRerouteUsage()
  NamedRerouteDeclaration(None=LinearInterpolate(A=Multiply(A=@n29, B=@n30) {constB=5}, Alpha=NamedRerouteUsage()) {constB=0}) {name=Emissive_Out, nodeColor=<rgba 0.9219 1 0 1>}
  n31 = ScalarParameter() {parameterName=FresnelMin}
  n32 = ScalarParameter() {defaultValue=0.8, parameterName=FresnelMax}
  n33 = MaterialFunctionCall(Power=ScalarParameter() {defaultValue=1, parameterName=FresnelPower}) {materialFunction=&Fresnel_Function}
  NamedRerouteDeclaration(None=Saturate(None=LinearInterpolate(A=@n31, B=@n32, Alpha=@n33.Result))) {name=Fresnel, nodeColor=<rgba 0 0.4609 1 1>}
 ]}
