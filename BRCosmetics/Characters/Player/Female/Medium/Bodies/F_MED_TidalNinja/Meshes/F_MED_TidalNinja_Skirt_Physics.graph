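# Physics asset for the Tidal Ninja female medium body's skirt, simulating left-side front, side and back panel bone chains.
# Each chain hangs from a base bone constrained to root, with limited swing and zero twist range between dynamic segments.
# The front chain inserts a second free-jointed base bone after its first segment.
# Thigh, calf, foot and left knee-pad capsules plus a ground-level root box provide collision shapes around the skirt.
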PhysicsAsset /BRCosmetics/Characters/Player/Female/Medium/Bodies/F_MED_TidalNinja/Meshes/F_MED_TidalNinja_Skirt_Physics
bodies {
 base_kneePad_l [{
   shapeType=Capsule
   length=2.6198
   radius=4.898
   shapeName=calf_l_capsule
   center=<0.8944 -2.5659 0.3049>
   rotation=<rot 4.4579 91.0365 -89.9192>
  }]
 base_skirt_back_lf [{shapeType=Capsule, length=1, radius=2, shapeName=auto_skirt_front_lf_capsule, rotation=<rot 0 0 90>}]
 base_skirt_front_lf [{shapeType=Capsule, length=1, radius=2, shapeName=auto_skirt_front_lf_capsule, rotation=<rot 0 0 90>}]
 base_skirt_front_lf_01 [{shapeType=Capsule, length=1, radius=2, shapeName=auto_skirt_front_lf_capsule, rotation=<rot 0 0 90>}]
 base_skirt_side_lf [{shapeType=Capsule, length=1, radius=2, shapeName=auto_skirt_side_lf_capsule, rotation=<rot 0 0 90>}]
 calf_l [
  {
   shapeType=Capsule
   length=42.9045
   radius=9.7831
   shapeName=calf_l_capsule
   center=<-31.4494 -4.1653 -1.4683>
   rotation=<rot 89.6966 0.0013 0.0015>
  }
  {
   shapeType=Capsule
   length=2.6198
   radius=5.8651
   shapeName=calf_l_capsule
   center=<-4.485 0.6252 0.5538>
   rotation=<rot 4.4377 84.4325 -90.432>
  }
 ]
 calf_r [
  {
   shapeType=Capsule
   length=43.6012
   radius=9.6091
   shapeName=calf_l_capsule
   center=<28.7027 2.4834 0.0306>
   rotation=<rot 89.6292 179.9987 179.9989>
  }
  {
   shapeType=Capsule
   length=2.6198
   radius=5.8651
   shapeName=calf_l_capsule
   center=<-4.5223 0.2419 0.5239>
   rotation=<rot 4.4377 84.4325 -90.432>
  }
 ]
 dyn_skirt_back_lf_00 [{shapeType=Capsule, length=6, radius=8.3041, shapeName=dyn_skirt_back_lf_01_capsule, center=<8 -5 -5>, rotation=<rot 90 -166.4056 -73.1015>}]
 dyn_skirt_back_lf_01 [{shapeType=Capsule, length=6, radius=8.3041, shapeName=dyn_skirt_back_lf_01_capsule, center=<8 -5 -5>, rotation=<rot 90 -168.5441 -75.24>}]
 dyn_skirt_back_lf_02 [{shapeType=Capsule, length=6, radius=8.3041, shapeName=dyn_skirt_back_lf_01_capsule, center=<8 -5 -5>, rotation=<rot 90 -171.6296 -78.3254>}]
 dyn_skirt_back_lf_03 [{
   shapeType=Capsule
   length=6
   radius=8.3041
   shapeName=dyn_skirt_back_lf_01_capsule
   center=<8 -5 -5.9673>
   rotation=<rot 90 -172.875 -79.5709>
  }]
 dyn_skirt_front_lf_00 [{shapeType=Capsule, length=1, radius=2, shapeName=auto_skirt_front_lf_capsule, rotation=<rot 0 0 90>}]
 dyn_skirt_front_lf_01 [{
   shapeType=Capsule
   length=6
   radius=2.2104
   shapeName=dyn_skirt_back_lf_01_capsule
   center=<8 3.3 -3>
   rotation=<rot 58.879 89.3773 174.5312>
  }]
 dyn_skirt_front_lf_02 [{
   shapeType=Capsule
   length=13.603
   radius=9.7633
   shapeName=dyn_skirt_back_lf_01_capsule
   center=<13.917 2.4567 -9.8671>
   rotation=<rot 57.8019 89.9998 176.8907>
  }]
 dyn_skirt_front_lf_03 [{
   shapeType=Capsule
   length=13.603
   radius=8.1587
   shapeName=dyn_skirt_back_lf_01_capsule
   center=<14.8 3.3 -8>
   rotation=<rot 61.6068 97.8554 -174.1898>
  }]
 dyn_skirt_side_lf_00 [{shapeType=Capsule, length=6, radius=9.4074, shapeName=dyn_skirt_back_lf_01_capsule, center=<8 -5 -5>, rotation=<rot 90 -161.4862 -70.6768>}]
 dyn_skirt_side_lf_01 [{shapeType=Capsule, length=6, radius=9.4074, shapeName=dyn_skirt_back_lf_01_capsule, center=<8 -5 -5>, rotation=<rot 90 180 -89.1906>}]
 dyn_skirt_side_lf_02 [{
   shapeType=Capsule
   length=20
   radius=9.4074
   shapeName=dyn_skirt_back_lf_01_capsule
   center=<9.1203 -5.0158 -4.4486>
   rotation=<rot 90 175.2364 -93.9542>
  }]
 dyn_skirt_side_lf_03 [{
   shapeType=Capsule
   length=23
   radius=8.9435
   shapeName=dyn_skirt_back_lf_01_capsule
   center=<9.1203 -5 -4.1>
   rotation=<rot 79.7941 89.1907 -179.9998>
  }]
 foot_l [{
   shapeType=Capsule
   length=16.7247
   radius=9.3674
   shapeName=foot_r_capsule
   center=<-2.3846 4.1029 -0.0222>
   rotation=<rot 0.7639 -2.7396 -88.6671>
  }]
 foot_r [{
   shapeType=Capsule
   length=16.7247
   radius=9.3674
   shapeName=foot_r_capsule
   center=<2.3846 -4.1029 -0.0222>
   rotation=<rot 0.7639 -2.7396 -88.6671>
  }]
 root [{shapeType=Box, extentX=204.1628, extentY=195.3963, extentZ=26.3202, shapeName=root_box, center=<0 0 -11.8908>, rotation=<rot 0 -90 0>}]
 thigh_l [{
   shapeType=Capsule
   length=30.8991
   radius=6.4723
   shapeName=thigh_l_capsule
   center=<-22.7 0.0977 0.0564>
   rotation=<rot 88.228 179.7204 179.1418>
  }]
 thigh_r [{
   shapeType=Capsule
   length=31.2673
   radius=7.897
   shapeName=thigh_l_capsule
   center=<22.6928 -0.6218 0.0567>
   rotation=<rot 88.228 179.7206 179.7215>
  }]
}
constraints [
 {
  bone1Name=calf_l
  bone2Name=thigh_l
  swing1LimitDegrees=45
  swing1Motion=Free
  swing2LimitDegrees=45
  swing2Motion=Free
  twistLimitDegrees=45
  twistMotion=Free
 }
 {bone1Name=dyn_skirt_back_lf_00, bone2Name=base_skirt_back_lf, swing1LimitDegrees=15, swing2LimitDegrees=15, twistLimitDegrees=0}
 {bone1Name=dyn_skirt_back_lf_01, bone2Name=dyn_skirt_back_lf_00, swing1LimitDegrees=10, swing2LimitDegrees=20, twistLimitDegrees=0}
 {bone1Name=dyn_skirt_back_lf_02, bone2Name=dyn_skirt_back_lf_01, swing1LimitDegrees=10, swing2LimitDegrees=20, twistLimitDegrees=0}
 {bone1Name=dyn_skirt_back_lf_03, bone2Name=dyn_skirt_back_lf_02, swing1LimitDegrees=10, swing2LimitDegrees=20, twistLimitDegrees=0}
 {bone1Name=dyn_skirt_front_lf_00, bone2Name=base_skirt_front_lf, swing1LimitDegrees=35, swing2LimitDegrees=35, twistLimitDegrees=0}
 {bone1Name=dyn_skirt_front_lf_01, bone2Name=base_skirt_front_lf_01, swing1LimitDegrees=40, swing2LimitDegrees=30, twistLimitDegrees=0}
 {bone1Name=dyn_skirt_front_lf_02, bone2Name=dyn_skirt_front_lf_01, swing1LimitDegrees=25, swing2LimitDegrees=20, twistLimitDegrees=0}
 {bone1Name=dyn_skirt_front_lf_03, bone2Name=dyn_skirt_front_lf_02, swing1LimitDegrees=25, swing2LimitDegrees=20, twistLimitDegrees=0}
 {bone1Name=dyn_skirt_side_lf_00, bone2Name=base_skirt_side_lf, swing1LimitDegrees=20, swing2LimitDegrees=10, twistLimitDegrees=0}
 {bone1Name=dyn_skirt_side_lf_01, bone2Name=dyn_skirt_side_lf_00, swing1LimitDegrees=20, swing2LimitDegrees=10, twistLimitDegrees=0}
 {bone1Name=dyn_skirt_side_lf_02, bone2Name=dyn_skirt_side_lf_01, swing1LimitDegrees=20, swing2LimitDegrees=10, twistLimitDegrees=0}
 {bone1Name=dyn_skirt_side_lf_03, bone2Name=dyn_skirt_side_lf_02, swing1LimitDegrees=20, swing2LimitDegrees=10, twistLimitDegrees=0}
 {
  bone1Name=calf_r
  bone2Name=thigh_r
  swing1LimitDegrees=45
  swing1Motion=Free
  swing2LimitDegrees=45
  swing2Motion=Free
  twistLimitDegrees=45
  twistMotion=Free
 }
 {
  bone1Name=thigh_l
  bone2Name=root
  swing1LimitDegrees=45
  swing1Motion=Free
  swing2LimitDegrees=45
  swing2Motion=Free
  twistLimitDegrees=45
  twistMotion=Free
 }
 {
  bone1Name=thigh_r
  bone2Name=root
  swing1LimitDegrees=45
  swing1Motion=Free
  swing2LimitDegrees=45
  swing2Motion=Free
  twistLimitDegrees=45
  twistMotion=Free
 }
 {
  bone1Name=base_skirt_front_lf
  bone2Name=root
  swing1LimitDegrees=45
  swing1Motion=Free
  swing2LimitDegrees=45
  swing2Motion=Free
  twistLimitDegrees=45
  twistMotion=Free
 }
 {
  bone1Name=base_skirt_side_lf
  bone2Name=root
  swing1LimitDegrees=45
  swing1Motion=Free
  swing2LimitDegrees=45
  swing2Motion=Free
  twistLimitDegrees=45
  twistMotion=Free
 }
 {
  bone1Name=base_skirt_back_lf
  bone2Name=root
  swing1LimitDegrees=45
  swing1Motion=Free
  swing2LimitDegrees=45
  swing2Motion=Free
  twistLimitDegrees=45
  twistMotion=Free
 }
 {
  bone1Name=foot_l
  bone2Name=calf_l
  swing1LimitDegrees=45
  swing1Motion=Free
  swing2LimitDegrees=45
  swing2Motion=Free
  twistLimitDegrees=45
  twistMotion=Free
 }
 {
  bone1Name=foot_r
  bone2Name=calf_r
  swing1LimitDegrees=45
  swing1Motion=Free
  swing2LimitDegrees=45
  swing2Motion=Free
  twistLimitDegrees=45
  twistMotion=Free
 }
 {
  bone1Name=base_skirt_front_lf_01
  bone2Name=dyn_skirt_front_lf_00
  swing1LimitDegrees=45
  swing1Motion=Free
  swing2LimitDegrees=45
  swing2Motion=Free
  twistLimitDegrees=45
  twistMotion=Free
 }
 {
  bone1Name=base_kneePad_l
  bone2Name=calf_l
  swing1LimitDegrees=45
  swing1Motion=Free
  swing2LimitDegrees=45
  swing2Motion=Free
  twistLimitDegrees=45
  twistMotion=Free
 }
]
