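# Character-part AnimBP for the F_MED_TidalNinja body that adds secondary motion to skirt, belt and gem bones.
# The copied parent pose runs through a Control Rig, RigidBody nodes with normal, DBNO and skydive physics assets, then AnimDynamics, constraints and additive bone offsets.
# When rigid body is disabled, crouch, knee-slide and baseball-slide flags select fixed skirt corrections.
# EventGraph sets those flags, Control Rig state values and gravity override each update.
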AnimBlueprint /BRCosmetics/Characters/Player/Female/Medium/Bodies/F_MED_TidalNinja/Meshes/F_MED_TidalNinja_AnimBP
target_skeleton=&/F_MED_TidalNinja_Skeleton
graphs [
 AnimGraph {nodes [
   n1 = "Transform (Modify) Bone - Bone: base_skirt_front_lf_01"(Alpha=1, Translation="0,0,0", Rotation="0,0,0", Scale="1,1,1") {
    rotationMode=BMM_Additive
    rotationSpace=BCS_BoneSpace
    translationMode=BMM_Additive
    translationSpace=BCS_BoneSpace
    alphaBoolBlend {blendOption=Linear}
    boneToModify {boneName=base_skirt_front_lf_01}
   }
   n2 = "Transform (Modify) Bone - Bone: dyn_skirt_front_lf_02"(ComponentPose=@n1, Alpha=1, Translation="0,0,0", Rotation="10,0,0", Scale="1,1,1") {
    rotationMode=BMM_Additive
    rotationSpace=BCS_BoneSpace
    translationMode=BMM_Additive
    translationSpace=BCS_BoneSpace
    alphaBoolBlend {blendOption=Linear}
    boneToModify {boneName=dyn_skirt_front_lf_02}
    rotation {pitch=10}
   }
   n3 = "Transform (Modify) Bone - Bone: dyn_skirt_front_lf_03"(ComponentPose=@n2, Alpha=1, Translation="0,0,0", Rotation="10,-30,0", Scale="1,1,1") {
    rotationMode=BMM_Additive
    rotationSpace=BCS_BoneSpace
    translationMode=BMM_Additive
    translationSpace=BCS_BoneSpace
    alphaBoolBlend {blendOption=Linear}
    boneToModify {boneName=dyn_skirt_front_lf_03}
    rotation {pitch=10, yaw=-30}
   }
   "Transform (Modify) Bone - Bone: dyn_skirt_side_lf_03"(ComponentPose=@n3, Alpha=1, Translation="-5,0,0", Rotation="0,0,0", Scale="1,1,1") {
    rotationMode=BMM_Additive
    rotationSpace=BCS_BoneSpace
    translationMode=BMM_Additive
    translationSpace=BCS_BoneSpace
    alphaBoolBlend {blendOption=Linear}
    boneToModify {boneName=dyn_skirt_side_lf_03}
    translation {x=-5}
   }
   n4 = "Transform (Modify) Bone - Bone: dyn_skirt_back_lf_01"(
    ComponentPose=Local To Component(LocalPose=Use cached pose 'CRCache_Setup'())
    Alpha=1
    Translation="4.350,-0.035,-0.257"
    Rotation="35.914,-0.014,0.013"
    Scale="1,1,1"
   ) {
    rotationMode=BMM_Additive
    rotationSpace=BCS_BoneSpace
    translationMode=BMM_Additive
    translationSpace=BCS_BoneSpace
    alphaBoolBlend {blendOption=Linear}
    boneToModify {boneName=dyn_skirt_back_lf_01}
    rotation=<rot 35.914 -0.014 0.013>
    translation=<4.35 -0.035 -0.257>
   }
   n5 = "Transform (Modify) Bone - Bone: dyn_skirt_side_lf_01"(ComponentPose=@n4, Alpha=1, Translation="0,0,0", Rotation="-30.818,-0.020,0.011", Scale="1,1,1") {
    rotationMode=BMM_Additive
    rotationSpace=BCS_BoneSpace
    translationMode=BMM_Additive
    translationSpace=BCS_BoneSpace
    alphaBoolBlend {blendOption=Linear}
    boneToModify {boneName=dyn_skirt_side_lf_01}
    rotation=<rot -30.818 -0.02 0.011>
   }
   n6 = "Transform (Modify) Bone - Bone: dyn_skirt_front_lf_01"(ComponentPose=@n5, Alpha=1, Translation="0,0,0", Rotation="0.039,43.153,0.129", Scale="1,1,1") {
    rotationMode=BMM_Additive
    rotationSpace=BCS_BoneSpace
    translationMode=BMM_Additive
    translationSpace=BCS_BoneSpace
    alphaBoolBlend {blendOption=Linear}
    boneToModify {boneName=dyn_skirt_front_lf_01}
    rotation=<rot 0.039 43.153 0.129>
   }
   n7 = Blend Poses by bool(
    BlendPose_0=Component To Local(ComponentPose=@n6)
    BlendPose_1=Use cached pose 'CRCache_Setup'()
    bActiveValue=Get CrouchNotDyn() -> (CrouchNotDyn:bool)
   ) {blendType=Linear}
   n8 = "Transform (Modify) Bone - Bone: dyn_skirt_back_lf_01"(
    ComponentPose=Local To Component(LocalPose=Use cached pose 'CRCache_Setup'())
    Alpha=1
    Translation="0,0,0"
    Rotation="67.729,0.000,0.000"
    Scale="1,1,1"
   ) {
    rotationMode=BMM_Additive
    rotationSpace=BCS_BoneSpace
    translationMode=BMM_Additive
    translationSpace=BCS_BoneSpace
    alphaBoolBlend {blendOption=Linear}
    boneToModify {boneName=dyn_skirt_back_lf_01}
    rotation=<rot 67.729 0 0>
   }
   n9 = "Transform (Modify) Bone - Bone: dyn_skirt_side_lf_01"(ComponentPose=@n8, Alpha=1, Translation="0,0,0", Rotation="-24.542,-15.735,4.594", Scale="1,1,1") {
    rotationMode=BMM_Additive
    rotationSpace=BCS_BoneSpace
    translationMode=BMM_Additive
    translationSpace=BCS_BoneSpace
    alphaBoolBlend {blendOption=Linear}
    boneToModify {boneName=dyn_skirt_side_lf_01}
    rotation=<rot -24.542 -15.735 4.594>
   }
   n10 = "Transform (Modify) Bone - Bone: dyn_skirt_front_lf_01"(ComponentPose=@n9, Alpha=1, Translation="0,0,0", Rotation="-39.861,43.048,0.168", Scale="1,1,1") {
    rotationMode=BMM_Additive
    rotationSpace=BCS_BoneSpace
    translationMode=BMM_Additive
    translationSpace=BCS_BoneSpace
    alphaBoolBlend {blendOption=Linear}
    boneToModify {boneName=dyn_skirt_front_lf_01}
    rotation=<rot -39.861 43.048 0.168>
   }
   n11 = Component To Local(ComponentPose=@n10)
   n12 = Get KneeSlideNotDyn() -> (KneeSlideNotDyn:bool)
   n13 = "Transform (Modify) Bone - Bone: dyn_skirt_back_lf_01"(
    ComponentPose=Local To Component(LocalPose=Use cached pose 'CRCache_Setup'())
    Alpha=1
    Translation="-3.875,-0.129,1.213"
    Rotation="65.931,62.270,56.322"
    Scale="1,1,1"
   ) {
    rotationMode=BMM_Additive
    rotationSpace=BCS_BoneSpace
    translationMode=BMM_Additive
    translationSpace=BCS_BoneSpace
    alphaBoolBlend {blendOption=Linear}
    boneToModify {boneName=dyn_skirt_back_lf_01}
    rotation=<rot 65.931 62.27 56.322>
    translation=<-3.875 -0.129 1.213>
   }
   n14 = "Transform (Modify) Bone - Bone: dyn_skirt_side_lf_01"(ComponentPose=@n13, Alpha=1, Translation="0,0,0", Rotation="3.979,14.989,37.409", Scale="1,1,1") {
    rotationMode=BMM_Additive
    rotationSpace=BCS_BoneSpace
    translationMode=BMM_Additive
    translationSpace=BCS_BoneSpace
    alphaBoolBlend {blendOption=Linear}
    boneToModify {boneName=dyn_skirt_side_lf_01}
    rotation=<rot 3.979 14.989 37.409>
   }
   n15 = "Transform (Modify) Bone - Bone: dyn_skirt_front_lf_01"(ComponentPose=@n14, Alpha=1, Translation="0,0,0", Rotation="-54.736,102.374,-18.514", Scale="1,1,1") {
    rotationMode=BMM_Additive
    rotationSpace=BCS_BoneSpace
    translationMode=BMM_Additive
    translationSpace=BCS_BoneSpace
    alphaBoolBlend {blendOption=Linear}
    boneToModify {boneName=dyn_skirt_front_lf_01}
    rotation=<rot -54.736 102.374 -18.514>
   }
   Output Pose(Result=Blend Poses by bool(
     BlendPose_0=Component To Local(ComponentPose=@n15)
     BlendPose_1=Blend Poses by bool(BlendPose_0=@n11, BlendPose_1=@n7, bActiveValue=@n12) {blendType=Linear}
     bActiveValue=Get BaseSlideNotDyn() -> (BaseSlideNotDyn:bool)
    ) {blendType=Linear}) {layerGroup=None}
   n16 = Control Rig!(
    Source=Copy Pose From Mesh() {bUseAttachedParent=true}
    Alpha
    notCrouching:double=Get notCrouching() -> (notCrouching:double)
    isSkyDiveFloating:double=Get isSkyDiveFloating() -> (isSkyDiveFloating:double)
    isMoving2D:double=Get isMoving2D() -> (isMoving2D:double)
    isSliding:double=Get isSliding() -> (isSliding:double)
    isBaseBallSliding:double=Get isBaseballSliding() -> (isBaseballSliding:double)
    isOnGrindRail:double=Get isOnGrindRail() -> (isOnGrindRail:double)
    isNotDBNO:double=Get isDBNO() -> (isDBNO:double)
    RigidBodyOn:bool=Get bRigidBodyEnabled()
   ) -> (Pose) {
    alphaBoolBlend {blendOption=Linear}
    controlRigAssetReference {blueprintRigClass=&/F_MED_TidalNinja_CtrlRig.F_MED_TidalNinja_CtrlRig_C}
   }
   n17 = "Transform (Modify) Bone - Bone: dyn_skirt_back_lf_01"!(
    ComponentPose=Local To Component(LocalPose=@n16)
    bAlphaBoolEnabled:bool=True
    Translation="0,0,0"
    Rotation="-77.523,-0.112,0.048"
    Scale="1,1,1"
   ) -> (Pose) {
    alphaInputType=Bool
    rotationMode=BMM_Additive
    rotationSpace=BCS_BoneSpace
    translationMode=BMM_Additive
    translationSpace=BCS_BoneSpace
    alphaBoolBlend {blendOption=Linear}
    boneToModify {boneName=dyn_skirt_back_lf_01}
    rotation=<rot -77.523 -0.112 0.048>
   }
   n18 = "Transform (Modify) Bone - Bone: dyn_skirt_back_side_lf_01"!(ComponentPose=@n17, bAlphaBoolEnabled:bool=True, Translation="0,0,0", Rotation="-73.584,-30.822,2.015", Scale="1,1,1") -> (Pose) {
    alphaInputType=Bool
    rotationMode=BMM_Additive
    rotationSpace=BCS_BoneSpace
    translationMode=BMM_Additive
    translationSpace=BCS_BoneSpace
    alphaBoolBlend {blendOption=Linear}
    boneToModify {boneName=dyn_skirt_back_side_lf_01}
    rotation=<rot -73.584 -30.8219 2.0149>
   }
   n19 = "Transform (Modify) Bone - Bone: dyn_skirt_back_lf_00"(ComponentPose=@n18, Alpha=1, Translation="0,0,0", Rotation="-15,-5,0", Scale="1,1,1") {
    rotationMode=BMM_Additive
    rotationSpace=BCS_BoneSpace
    translationMode=BMM_Additive
    translationSpace=BCS_BoneSpace
    alphaBoolBlend {blendOption=Linear}
    boneToModify {boneName=dyn_skirt_back_lf_00}
    rotation {pitch=-15, yaw=-5}
   }
   n20 = "Transform (Modify) Bone - Bone: base_skirt_side_lf"(ComponentPose=@n19, Alpha=1, Translation="0,0,0", Rotation="-40,10,0", Scale="1,1,1") {
    rotationMode=BMM_Additive
    rotationSpace=BCS_BoneSpace
    translationMode=BMM_Additive
    translationSpace=BCS_BoneSpace
    alphaBoolBlend {blendOption=Linear}
    boneToModify {boneName=base_skirt_side_lf}
    rotation {pitch=-40, yaw=10}
   }
   Save cached pose 'CR_Start'(Pose=Component To Local(ComponentPose=@n20)) {cacheName=CR_Start}
   n21 = RigidBody!(
    ComponentPose=Local To Component(LocalPose=Use cached pose 'CR_Start'())
    Alpha
    OverrideWorldGravity:Vector=Get Gravity Override()
    ExternalForce
   ) -> (Pose) {
    bClampLinearTranslationLimitToRefPose=true
    bOverrideWorldGravity=true
    overridePhysicsAsset=&/F_MED_TidalNinja_Skirt_Physics
    simulationSpace=BaseBoneSpace
    alphaBoolBlend {blendOption=Linear}
    baseBoneRef {boneName=pelvis}
    componentAppliedLinearAccClamp=<3000 3000 3000>
    simSpaceSettings {
     dampingAlpha=0
     maxAngularAcceleration=400
     maxAngularVelocity=100
     maxLinearAcceleration=3200
     maxLinearVelocity=800
     velocityScaleZ=0.5
     worldAlpha=1
     externalLinearDragV=<0.6 0.6 0.6>
    }
   }
   n22 = Component To Local(ComponentPose=@n21)
   n23 = RigidBody(ComponentPose=Local To Component(LocalPose=Use cached pose 'CR_Start'()), OverrideWorldGravity=Get Gravity Override()) {
    bClampLinearTranslationLimitToRefPose=true
    bOverrideWorldGravity=true
    overridePhysicsAsset=&/F_MED_TidalNinja_Skirt_DBNO_Physics
    simulationSpace=BaseBoneSpace
    alphaBoolBlend {blendOption=Linear}
    baseBoneRef {boneName=pelvis}
    componentAppliedLinearAccClamp=<3000 3000 3000>
    simSpaceSettings {
     dampingAlpha=0
     maxAngularAcceleration=400
     maxAngularVelocity=100
     maxLinearAcceleration=3200
     maxLinearVelocity=800
     velocityScaleZ=0.7
     worldAlpha=1
     externalLinearDragV=<0.6 0.6 0.6>
    }
   }
   n24 = Component To Local(ComponentPose=@n23)
   n25 = RigidBody(ComponentPose=Local To Component(LocalPose=Use cached pose 'CR_Start'()), OverrideWorldGravity=Get Gravity Override()) {
    bClampLinearTranslationLimitToRefPose=true
    bOverrideWorldGravity=true
    overridePhysicsAsset=&/F_MED_TidalNinja_Skirt_Skydive_Physics
    simulationSpace=BaseBoneSpace
    alphaBoolBlend {blendOption=Linear}
    baseBoneRef {boneName=pelvis}
    componentAppliedLinearAccClamp=<0 0 0>
    simSpaceSettings {
     dampingAlpha=0
     maxAngularAcceleration=200
     maxAngularVelocity=100
     maxLinearAcceleration=800
     maxLinearVelocity=50
     velocityScaleZ=0.6
     worldAlpha=0.3
     externalLinearDragV=<0.2 0.2 0.2>
    }
   }
   Save cached pose 'CR_RBAN'(Pose=Blend Poses by bool(
     BlendPose_0=Component To Local(ComponentPose=@n25)
     BlendPose_1=Blend Poses by bool(BlendPose_0=@n24, BlendPose_1=@n22) {blendType=Linear}
     bActiveValue=Get bIsSkydiveFloating()
    ) {blendType=Linear}) {cacheName=CR_RBAN}
   n26 = "Transform (Modify) Bone - Bone: base_skirt_side_r"(ComponentPose=Local To Component(LocalPose=Use cached pose 'CR_RBAN'()), Alpha=1, Translation="0,0,0.8", Rotation="8,0,0", Scale="1,1,1") {
    rotationMode=BMM_Additive
    rotationSpace=BCS_BoneSpace
    scaleMode=BMM_Additive
    translationMode=BMM_Additive
    translationSpace=BCS_BoneSpace
    alphaBoolBlend {blendOption=Linear}
    alphaScaleBiasClamp {bInterpResult=true}
    boneToModify {boneName=base_skirt_side_r}
    rotation {pitch=8}
    translation {z=0.8}
   }
   n27 = Anim Dynamics - Bone: dyn_skirt_side_r_01(ComponentPose=@n26, Alpha=1, ExternalForce="0,0,0") {
    angularDampingOverride=1
    angularSpringConstant=20000
    bAngularSpring=true
    bOverrideAngularDamping=true
    bUseGravityOverride=true
    gravityScale=0.35
    simulationSpace=BoneRelative
    alphaBoolBlend {blendOption=Linear}
    boundBone {boneName=dyn_skirt_side_r_01}
    componentAppliedLinearAccClamp=<3000 3000 3000>
    componentLinearAccScale=<1 1 1>
    componentLinearVelScale=<1 1 1>
    gravityOverride {z=-980}
    physicsBodyDefinitions [{
      boundBone {boneName=dyn_skirt_side_r_01}
      boxExtents=<20 30 30>
      constraintSetup {angularLimitsMax=<0 0 0>, angularLimitsMin=<0 -20 0>, angularTarget=<1 0 0>}
      localJointOffset=<4 0 0>
     }]
    relativeSpaceBone {boneName=thigh_r}
   }
   n28 = Anim Dynamics - Bone: dyn_skirt_side_r_02(ComponentPose=@n27, Alpha=1, ExternalForce="0,0,0") {
    angularDampingOverride=1
    angularSpringConstant=20000
    bAngularSpring=true
    bOverrideAngularDamping=true
    bUseGravityOverride=true
    gravityScale=0.35
    simulationSpace=RootRelative
    alphaBoolBlend {blendOption=Linear}
    boundBone {boneName=dyn_skirt_side_r_02}
    componentAppliedLinearAccClamp=<3000 3000 3000>
    componentLinearAccScale=<1 1 1>
    componentLinearVelScale=<1 1 1>
    gravityOverride {z=-980}
    physicsBodyDefinitions [{
      boundBone {boneName=dyn_skirt_side_r_02}
      boxExtents=<20 30 30>
      constraintSetup {angularLimitsMax=<0 0 3>, angularLimitsMin=<0 -15 -3>, angularTarget=<1 0 0>}
      localJointOffset=<4 0 0>
     }]
    relativeSpaceBone {boneName=thigh_r}
   }
   n29 = "Constraint - belt_base (PPPPP)"!(
    ComponentPose=@n28
    Alpha=1
    ConstraintWeights_0=0.7509
    ConstraintWeights_1:float=0.2141
    ConstraintWeights_2:float=0.0217
    ConstraintWeights_3:float=0.0074
    ConstraintWeights_4:float=0.0059
   ) -> (Pose) {
    alphaBoolBlend {blendOption=Linear}
    boneToModify {boneName=belt_base}
    constraintSetup [
     {targetBone {boneName=pelvis}}
     {targetBone {boneName=spine_02}}
     {targetBone {boneName=spine_04}}
     {targetBone {boneName=spine_03}}
     {targetBone {boneName=spine_01}}
    ]
    constraintWeights [1,1,1,1,1]
   }
   n30 = Anim Dynamics - Bone: belt_dyn(ComponentPose=@n29, Alpha=1, ExternalForce="0,0,0") {
    angularDampingOverride=1
    angularSpringConstant=20000
    bAngularSpring=true
    bOverrideAngularDamping=true
    bUseGravityOverride=true
    gravityScale=0.35
    simulationSpace=RootRelative
    alphaBoolBlend {blendOption=Linear}
    boundBone {boneName=belt_dyn}
    componentAppliedLinearAccClamp=<10000 10000 10000>
    componentLinearAccScale=<1 1 1>
    componentLinearVelScale=<1 1 1>
    gravityOverride {z=-980}
    physicsBodyDefinitions [{
      boundBone {boneName=belt_dyn}
      boxExtents=<20 30 30>
      constraintSetup {angularLimitsMax=<15 15 15>, angularLimitsMin=<-15 -15 -15>, angularTarget=<1 0 0>}
      localJointOffset=<4 0 0>
     }]
   }
   n31 = "Constraint - base_gem (PP)"!(ComponentPose=@n30, Alpha=1, ConstraintWeights_0=0.8039, ConstraintWeights_1:float=0.1961) -> (Pose) {
    alphaBoolBlend {blendOption=Linear}
    boneToModify {boneName=base_gem}
    constraintSetup [{targetBone {boneName=spine_05}}, {targetBone {boneName=spine_04}}]
    constraintWeights [1,1]
   }
   n32 = "Transform (Modify) Bone - Bone: base_skirt_front_lf_01"(ComponentPose=@n31, Alpha=1, Translation="0,0,0", Rotation="20.679,8.095,21.570", Scale="1,1,1") {
    rotationMode=BMM_Additive
    rotationSpace=BCS_BoneSpace
    translationMode=BMM_Additive
    translationSpace=BCS_BoneSpace
    alphaBoolBlend {blendOption=Linear}
    boneToModify {boneName=base_skirt_front_lf_01}
    rotation=<rot 20.679 8.095 21.57>
   }
   n33 = "Transform (Modify) Bone - Bone: dyn_skirt_front_lf_02"(ComponentPose=@n32, Alpha=1, Translation="0,0,0", Rotation="-21.778,-29.970,-0.031", Scale="1,1,1") {
    rotationMode=BMM_Additive
    rotationSpace=BCS_BoneSpace
    translationMode=BMM_Additive
    translationSpace=BCS_BoneSpace
    alphaBoolBlend {blendOption=Linear}
    boneToModify {boneName=dyn_skirt_front_lf_02}
    rotation=<rot -21.778 -29.97 -0.0309>
   }
   n34 = "Transform (Modify) Bone - Bone: dyn_skirt_front_lf_03"(ComponentPose=@n33, Alpha=1, Translation="0,0,0", Rotation="15.487,-39.974,-0.009", Scale="1,1,1") {
    rotationMode=BMM_Additive
    rotationSpace=BCS_BoneSpace
    translationMode=BMM_Additive
    translationSpace=BCS_BoneSpace
    alphaBoolBlend {blendOption=Linear}
    boneToModify {boneName=dyn_skirt_front_lf_03}
    rotation=<rot 15.487 -39.974 -0.009>
   }
   n35 = "Transform (Modify) Bone - Bone: dyn_skirt_side_lf_01"(ComponentPose=@n34, Alpha=1, Translation="0,0,0", Rotation="-21.663,-30.369,-17.790", Scale="1,1,1") {
    rotationMode=BMM_Additive
    rotationSpace=BCS_BoneSpace
    translationMode=BMM_Additive
    translationSpace=BCS_BoneSpace
    alphaBoolBlend {blendOption=Linear}
    boneToModify {boneName=dyn_skirt_side_lf_01}
    rotation=<rot -21.663 -30.369 -17.79>
   }
   n36 = "Transform (Modify) Bone - Bone: dyn_skirt_back_lf_01"!(ComponentPose=@n35, Alpha:float=1, Translation="0,0,5", Rotation="-10,0,0", Scale="1,1,1") -> (Pose) {
    rotationMode=BMM_Additive
    rotationSpace=BCS_BoneSpace
    translationMode=BMM_Additive
    translationSpace=BCS_BoneSpace
    alphaBoolBlend {blendOption=Linear}
    boneToModify {boneName=dyn_skirt_back_lf_01}
    rotation {pitch=-10}
    translation {z=5}
   }
   n37 = "Transform (Modify) Bone - Bone: base_skirt_front_lf_01"(ComponentPose=@n36, Alpha=1, Translation="0,0,0", Rotation="0.021,2.330,0.034", Scale="1,1,1") {
    rotationMode=BMM_Additive
    rotationSpace=BCS_BoneSpace
    translationMode=BMM_Additive
    translationSpace=BCS_BoneSpace
    alphaBoolBlend {blendOption=Linear}
    boneToModify {boneName=base_skirt_front_lf_01}
    rotation=<rot 0.021 2.33 0.034>
   }
   n38 = "Transform (Modify) Bone - Bone: dyn_skirt_side_lf_01"(ComponentPose=@n37, Alpha=1, Translation="-0.487,4.296,2.699", Rotation="-20,20,0", Scale="1,1,1") {
    rotationMode=BMM_Additive
    rotationSpace=BCS_BoneSpace
    translationMode=BMM_Additive
    translationSpace=BCS_BoneSpace
    alphaBoolBlend {blendOption=Linear}
    boneToModify {boneName=dyn_skirt_side_lf_01}
    rotation {pitch=-20, yaw=20}
    translation=<-0.487 4.296 2.699>
   }
   n39 = "Transform (Modify) Bone - Bone: dyn_skirt_side_lf_03"(ComponentPose=@n38, Alpha=1, Translation="-7,0,0", Rotation="0,0,0", Scale="1,1,1") {
    rotationMode=BMM_Additive
    rotationSpace=BCS_BoneSpace
    translationMode=BMM_Additive
    translationSpace=BCS_BoneSpace
    alphaBoolBlend {blendOption=Linear}
    boneToModify {boneName=dyn_skirt_side_lf_03}
    translation {x=-7}
   }
   Save cached pose 'CRCache_Setup'(Pose=Component To Local(ComponentPose=@n39)) {cacheName=CRCache_Setup}
  ]}
 EventGraph {nodes [
   n1 = Event BlueprintUpdateAnimation()
   > Set Delta Time X(Delta Time X=@n1.DeltaTimeX)
   > n2 = Sequence!(execute) -> (then_0, then_1, then_2:Exec, then_3:Exec, then_4:Exec, then_5:Exec, then_6:Exec, then_7:Exec, then_8:Exec, then_9:Exec, then_10:Exec)
   > n3 = Gravity Override State Function DEPRECATED(CharacterPartAnimInstance=Self-Reference(), GameTimeInSeconds=Get Delta Time X(), GravityOverrideParams=Get Gravity Override Params())
   > Set Gravity Override(Gravity Override=@n3)
   n4 = Branch(execute=@n2.then_1, Condition=Get bIsCrouching())
   > Set notCrouching(notCrouching:double=0) -> (Output_Get:double)
   Set notCrouching(execute=@n4.else, notCrouching=1)
   n5 = Branch(execute=@n2.then_2, Condition=Get bIsSkydiveFloating())
   > Set isSkyDiveFloating(isSkyDiveFloating:double=1) -> (Output_Get:double)
   Set isSkyDiveFloating(execute=@n5.else, isSkyDiveFloating=0)
   n6 = Get bIsSkyDiving()
   n7 = Branch(execute=@n2.then_3, Condition=AND Boolean(A=Get bIsMoving2D(), B=NOT Boolean(A=@n6)))
   > Set isMoving2D(isMoving2D:double=1) -> (Output_Get:double)
   Set isMoving2D(execute=@n7.else, isMoving2D=0)
   n8 = Branch(execute=@n2.then_6, Condition=Get bIsOnGrindRail())
   > Set isOnGrindRail(isOnGrindRail:double=1) -> (Output_Get:double)
   Set isOnGrindRail(execute=@n8.else, isOnGrindRail=0)
   n9 = Branch(execute=@n2.then_7, Condition=Get bIsDBNO())
   > Set isDBNO(isDBNO:double=0) -> (Output_Get:double)
   Set isDBNO(execute=@n9.else, isDBNO=1)
   n10 = Get bIsMoving2D()
   n11 = Branch(execute=@n2.then_8, Condition=AND Boolean(A=Get bIsCrouching(), B=NOT Boolean(A=@n10)))
   > Set isCrouching(isCrouching:bool=true) -> (Output_Get:bool)
   Set isCrouching(execute=@n11.else, isCrouching=false)
   n12 = Get bIsPlayingEmote()
   n13 = Get bIsMoving2D()
   n14 = Get bIsSkyDiving()
   n15 = Branch(execute=@n2.then_9, Condition=AND Boolean!(A=NOT Boolean(A=@n13), B=NOT Boolean(A=@n14), C:bool=NOT Boolean(A=@n12)) -> (ReturnValue))
   > Set NotFEIdle(NotFEIdle:double=0) -> (Output_Get:double)
   Set NotFEIdle(execute=@n15.else, NotFEIdle=1)
   n16 = Get bIsBaseballSliding()
   Set isBaseballSliding(
    execute=@n2.then_5
    isBaseballSliding:double=FInterpTo(Current=Get isBaseballSliding(), Target="To Float (Boolean)"(InBool=@n16), DeltaTime=Get Delta Time X(), InterpSpeed=10)
   ) -> (Output_Get:double)
   n17 = Get bIsKneeSliding()
   Set isSliding(
    execute=@n2.then_4
    isSliding:double=FInterpTo(Current=Get isSliding(), Target="To Float (Boolean)"(InBool=@n17), DeltaTime=Get Delta Time X(), InterpSpeed=5)
   ) -> (Output_Get:double)
   n18 = Branch(execute=@n2.then_10, Condition=Get bRigidBodyEnabled())
   > Set CrouchNotDyn(CrouchNotDyn:bool=false) -> (Output_Get:bool)
   > Set KneeSlideNotDyn(KneeSlideNotDyn:bool=false) -> (Output_Get:bool)
   > Set BaseSlideNotDyn(BaseSlideNotDyn:bool=false) -> (Output_Get:bool)
   n19 = Get bIsSliding()
   Set CrouchNotDyn(execute=@n18.else, CrouchNotDyn=AND Boolean!(A=Get bIsCrouching(), B=NOT Boolean(A=@n19)) -> (ReturnValue))
   > Set KneeSlideNotDyn(KneeSlideNotDyn=Get bIsKneeSliding())
   > Set BaseSlideNotDyn(BaseSlideNotDyn=Get bIsBaseballSliding())
  ]}
]
variables [
 BaseSlideNotDyn {type=boolean}
 CrouchNotDyn {type=boolean}
 Delta Time X {type=number}
 Gravity Override {type=Vector}
 Gravity Override Params {type=GravityOverrideParamsStruct}
 isBaseballSliding {type=number}
 isCrouching {type=boolean}
 isDBNO {type=number}
 isMoving2D {type=number}
 isOnGrindRail {type=number}
 isSkyDiveFloating {type=number}
 isSliding {type=number}
 KneeSlideNotDyn {type=boolean}
 notCrouching {type=number}
 NotFEIdle {type=number}
]
